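# Master material for Daisy rock props in the Artemis environment, layering world-aligned moss over baked rock textures with optional random tinting.
# An upward-normal mask places the moss, which has Fresnel fuzzy edges and a color curve atlas feeding a detail reroute.
# After a detail texture pass, the base optionally height-blends into the Apollo landscape runtime virtual textures near the ground.
# Shared functions then add snow progression, destructible damage and Creative move-tool effects.
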Material /Game/Environments/Artemis/Props/Daisy/Daisy_Rocks/Materials/M_DaisyMossRocksMaster
graph {
 properties {bUseMaterialAttributes=true, bUsedWithNanite=true, pixelDepthOffsetMode=PDOM_Legacy, usedShadingModels=Default Lit}
 nodes [
  n1 = RuntimeVirtualTextureSampleParameter!(UVs, World Position, Mip Level=Constant() {r=8}) {
   materialType=BaseColor_Normal_Specular_YCoCg
   mipValueMode=RVTMVM_MipLevel
   parameterName=RVTColor
   virtualTexture=/Game/Athena/Apollo/Maps/Landscape/VT/RVT_Apollo_Landscape
  }
  n2 = RuntimeVirtualTextureSampleParameter!(UVs, World Position) {materialType=WorldHeight, parameterName=RVTHeight, virtualTexture=/Game/Athena/Apollo/Maps/Landscape/VT/RVT_Apollo_Landscape_Height}
  n3 = WorldPosition()
  n4 = ObjectRadius()
  n5 = VertexNormalWS()
  n6 = LinearInterpolate(
   A=ScalarParameter() {defaultValue=0.3, group=RVT, parameterName=Blend RVT North}
   B=ScalarParameter() {defaultValue=0.2, group=RVT, parameterName=Blend RVT}
   Alpha=Saturate(None=ComponentMask(None=@n5) {r=true})
  )
  n7 = Saturate(None=SphereMask(
    A=@n2.WorldHeight
    B=ComponentMask(None=@n3.XYZ) {b=true}
    Radius=Multiply(A=@n4, B=@n6)
    Hardness=ScalarParameter() {group=RVT, parameterName=BlendRVTContrast}
   ))
  n8 = NamedRerouteUsage()
  n9 = TextureSampleParameter2D() {
   group=BakedTextures
   parameterName=BakedDiffuse
   samplerType=SAMPLERTYPE_LinearColor
   texture=/Game/Global/Textures/Default/Blanks/T_BlackLinearColor
  }
  n10 = TextureSampleParameter2D() {
   group=BakedTextures
   parameterName=Baked Spec
   samplerType=SAMPLERTYPE_Masks
   texture=/Game/Global/Textures/Default/Blanks/Temp_specular_Mask
  }
  n11 = TextureSampleParameter2D() {group=BakedTextures, parameterName=Baked Normal, samplerType=SAMPLERTYPE_Normal, texture=/Game/Athena/WinterArt/Textures/Cracks_N}
  n12 = SetMaterialAttributes!(MaterialAttributes, Base Color=@n9.RGB, Metallic=@n10.G, Specular=@n10.G, Roughness=@n10.G, Normal=@n11.RGB) {attributeSetTypes [
    #1
    #2
    #3
    #4
    #5
   ]}
  n13 = GetMaterialAttributes(None=@n12) {attributeGetTypes [#1, #5]}
  n14 = VectorParameter() {group=Colors, parameterName=RandomColor_A, defaultValue=<rgba 0.5554 0.5898 0.6094 1>}
  n15 = VectorParameter() {group=Colors, parameterName=RandomColor_B, defaultValue=<rgba 0.4792 0.4429 0.3544 1>}
  n16 = ScalarParameter() {group=Colors, parameterName=RockRandomLerpValue}
  n17 = StaticSwitchParameter(True=Multiply(A=@n13.Base Color, B=LinearInterpolate(A=@n14.RGB, B=@n15.RGB, Alpha=@n16)), False=@n13.Base Color) {defaultValue=true, group=Colors, parameterName=RandomColors}
  n18 = Constant() {r=0.01}
  n19 = NamedRerouteUsage()
  n20 = VectorParameter() {parameterName=MossTintColor, defaultValue=<rgba 0.7708 0.5259 0.5872 1>}
  n21 = Saturate(None=Add(A=Multiply(A=@n18, B=@n19), B=Multiply(A=@n19, B=@n20.RGB)))
  n22 = ScalarParameter() {defaultValue=0.25, group=Moss, parameterName=Moss_CoreDarkening}
  n23 = ScalarParameter() {defaultValue=1, group=Moss, parameterName=Moss_FuzzyEdgeBrightness}
  n24 = TextureObjectParameter() {
   group=Top
   parameterName=Normals Top
   samplerType=SAMPLERTYPE_Normal
   sortPriority=6
   texture=/Game/Environments/Artemis/Sets/Terrain/Textures/T_Artemis_Frontier_Moss_01_N
  }
  n25 = ScalarParameter() {defaultValue=1000, group=Moss, parameterName=Scale Top, sortPriority=3}
  n26 = MaterialFunctionCall(TextureObject=@n24, TextureSize=@n25, WorldSpace=StaticBool(), Normal=VertexNormalWS()) {materialFunction=&WorldAlignedNormal}
  n27 = MaterialFunctionCall(Normal=@n26.Z Texture, Flatness=ScalarParameter() {group=Moss, parameterName=MossNormalFlatten}) {materialFunction=&FlattenNormal}
  n28 = Constant3Vector() {constant=<rgba 1 1 0 0>}
  n29 = Add(
   A=Multiply(A=Multiply(A=@n27.Result, B=@n28), B=ScalarParameter() {defaultValue=1, group=Moss, parameterName=Moss Normals Value})
   B=MaterialFunctionCall(Normal=@n13.Normal, Flatness=ScalarParameter() {defaultValue=1, group=Moss, parameterName=FlattenBAsweNormal}) {materialFunction=&FlattenNormal}.Result
  )
  n30 = Power(
   Base=Fresnel(ExponentIn=ScalarParameter() {defaultValue=2, group=Moss, parameterName=FuzzExponent}, Normal=Transform(None=@n29))
   Exp=ScalarParameter() {defaultValue=2, group=Moss, parameterName=FuzzEdgeClamp}
  )
  n31 = StaticBool() {value=true}
  n32 = TextureObjectParameter() {
   group=Top
   parameterName=Spec Top
   samplerType=SAMPLERTYPE_Masks
   sortPriority=4
   texture=/Game/Environments/Artemis/Sets/Terrain/Textures/T_Artemis_Frontier_Moss_01_S
  }
  n33 = WorldPosition()
  n34 = MaterialFunctionCall(Export Float 4=@n31, TextureObject=@n32, TextureSize=@n25, WorldPosition=@n33.XYZ) {materialFunction=/Game/Weapons/WeaponSkins/Wraps/MaterialFunctions/WorldAlignedTexture}
  n35 = SetMaterialAttributes!(
   MaterialAttributes
   Base Color=LinearInterpolate(A=Multiply(A=@n21, B=@n22), B=Multiply(A=@n21, B=@n23), Alpha=Clamp(None=@n30))
   Specular=LinearInterpolate(
    A=ScalarParameter() {defaultValue=0.1, group=Moss, parameterName=Moss_Min_Specular}
    B=ScalarParameter() {defaultValue=0.5, group=Moss, parameterName=Moss_Max_Specular}
    Alpha=ComponentMask(None=@n34.XYZ Texture) {r=true}
   )
   Roughness=LinearInterpolate(
    A=ScalarParameter() {defaultValue=0.75, group=Moss, parameterName=Moss_Min_Roughness}
    B=ScalarParameter() {defaultValue=1, group=Moss, parameterName=Moss_Max_Roughness}
    Alpha=ComponentMask(None=@n34.XYZ Texture) {b=true}
   )
   Normal=@n29
  ) {attributeSetTypes [
    #1
    #3
    #4
    #5
   ]}
  n36 = ComponentMask(None=VertexNormalWS()) {b=true}
  n37 = ScalarParameter() {defaultValue=2, group=Transition, parameterName=EdgeFalloffStrength}
  n38 = Clamp(None=Power(Base=Multiply(A=@n36, B=@n37), Exp=ScalarParameter() {defaultValue=2, group=Transition, parameterName=EdgeFalloffExp}))
  n39 = Add(A=Add(A=NamedRerouteUsage(), B=@n38))
  n40 = Add(A=Multiply(B=OneMinus(None=@n38)) {constA=2}, B=Constant())
  n41 = Saturate(None=ComponentMask(None=Subtract(A=@n39, B=@n40)) {r=true})
  n42 = MaterialFunctionCall(In=@n41, Contrast=ScalarParameter() {defaultValue=1, group=Transition, parameterName=Contrast}) {materialFunction=&CheapContrast}
  n43 = MaterialFunctionCall!(In=BlendMaterialAttributes(
    A=SetMaterialAttributes!(MaterialAttributes=@n13.MaterialAttributes, Base Color=@n17, Normal=@n13.MaterialAttributes) {attributeSetTypes [#1, #5]}
    B=@n35
    Alpha=StaticSwitchParameter(True=@n42.Result, False=@n41) {parameterName=UseCheapContrast}
   )) {materialFunction=/Game/Environments/Apollo/VS/Materials/MF_DetailTexture}
  n44 = MaterialFunctionCall(
   NoRVT=@n43.Result
   RVT=BlendMaterialAttributes(
    A=@n43.Result
    B=SetMaterialAttributes!(MaterialAttributes, Base Color=@n1.BaseColor, Specular=@n1.BaseColor, Roughness=@n1.BaseColor) {attributeSetTypes [#1, #3, #4]}
    Alpha=MaterialFunctionCall(Transition Phase=@n7, Height Texture=@n8) {materialFunction=&HeightLerp}.Alpha
   ) {vertexAttributeBlendType=UseA}
  ) {materialFunction=&FN_RVT_Switch_Material}
  n45 = MaterialFunctionCall!(
   IsStaticSnow
   UseAlternateStaticSnowMask
   AlternateStaticSnowMask
   "UseRoad&&OcclusionMask"=StaticBool()
   IsForGrassMeshToPushDown
   AdditionalMaskToRemoveSnow
   IsFallingSnowFX
   UseMinimapCoords
   IncreaseSnowUsingSlope
   UseSmoothBorder
   LerpHighM
   TemporalHeightLerpContrast
   IsLobbyIsland
   UseBR-IslandMask
   IsTree
   IsForRenderTarget
   UseSparkle
   UseSnowNormals
   UseAdditionalMaskToRemoveSnow
   IsTangentSpaceInput
   UseVertexNormalsForBasicSnow
   UseTangentSnowNormals
   Material=ShadingPathSwitch(Default=StaticSwitchParameter(True=@n43.Result, False=@n44.Result) {group=RVT, parameterName=NoRVT})
  ) {materialFunction=/Game/Athena/Apollo/Environments/Landscape/Materials/Biome/UniqueBiomeMaterialFunctions/MF_Apollo_Snow_Progression}
  n46 = StaticBool() {value=true}
  n47 = StaticBool()
  MaterialFunctionCall(MA In=MaterialFunctionCall(
    Material_In=@n45.Material
    UseCrackedDamageTextures=StaticBool()
    UseAdditiveWPO=StaticSwitchParameter(True=@n46, False=@n47) {parameterName=Use Additive WPO}
   ) {materialFunction=&MF_FN_DestructibleDamage}.Material_Out) {materialFunction=&MF_CreativeMoveTool}
  n48 = ComponentMask(None=VertexNormalWS()) {b=true}
  n49 = WorldPosition()
  n50 = ActorPositionWS()
  n51 = ComponentMask(None=Normalize(VectorInput=Subtract(A=@n49.XYZ, B=@n50))) {b=true}
  n52 = ScalarParameter() {group=Moss, parameterName=Blend2NormalDirections}
  n53 = Multiply(
   A=Add(A=LinearInterpolate(A=@n48, B=@n51, Alpha=@n52), B=ScalarParameter() {group=Moss, parameterName=Mossoffset})
   B=ScalarParameter() {defaultValue=2, group=Moss, parameterName=MossAmount}
  )
  n54 = LinearInterpolate(
   A=ScalarParameter() {defaultValue=-1.1814, group=Moss, parameterName=MossColorEdge_1}
   B=ScalarParameter() {defaultValue=0.8762, group=Moss, parameterName=MossColorEdge_2}
   Alpha=MaterialFunctionCall(
    Transition Phase=Saturate(None=@n53)
    Height Texture=NamedRerouteUsage()
    Contrast=ScalarParameter() {defaultValue=3.3398, group=Moss, parameterName=MossContrastEdge}
   ) {materialFunction=&HeightLerp}.Lerp Alpha No Contrast
  )
  MaterialFunctionCall(
   Normal=@n26.Z Texture
   Flatness=LinearInterpolate(
    A=ScalarParameter() {defaultValue=-0.2286, group=Moss, parameterName=MossNormalFlatten_Edge}
    B=ScalarParameter() {group=Moss, parameterName=MossNormalFlatten}
    Alpha=Saturate(None=@n54)
   )
  ) {materialFunction=&FlattenNormal}
  MaterialFunctionCall!(In=@n45.Material) {materialFunction=/Game/Athena/WinterArt/Materials/MF_GameWideShared}
  n55 = TextureObjectParameter() {
   group=Top
   parameterName=Diffuse Top
   samplerType=SAMPLERTYPE_Masks
   sortPriority=4
   texture=/Game/Environments/Artemis/Sets/Terrain/Textures/T_Artemis_Frontier_Moss_01_D
  }
  n56 = MaterialFunctionCall(Export Float 4=@n31, TextureObject=@n55, TextureSize=@n25, WorldPosition=@n33.XYZ) {materialFunction=&WorldAlignedTexture}
  NamedRerouteDeclaration(None=ComponentMask(None=@n56.XYZ Texture) {a=true}) {name=MossHeight, nodeColor=<rgba 0 0.8594 1 1>}
  n57 = CurveAtlasRowParameter(CurveTime=LinearInterpolate(A=ScalarParameter() {defaultValue=0.15, parameterName=MossMinColVal}, Alpha=ComponentMask(None=@n56.XYZ Texture) {r=true}) {constA=0.25}) {
   atlas=/Game/Athena/Apollo/Environments/Landscape/Materials/Frontier/FrontierColorAtlas
   curve=/Game/Athena/Apollo/Environments/Landscape/Materials/Frontier/FrontierMossDetailColor
   parameterName=MossColorCurve
  }
  NamedRerouteDeclaration(None=@n57) {name=MossBaseColorDetail, nodeColor=<rgba 1 0 0.6562 1>}
  Reroute()
  Saturate()
  SphereMask(A=@n36, B=Constant() {r=1}, Radius=ScalarParameter() {defaultValue=1, group=Transition, parameterName=MossCoverage}, Hardness=@n37)
 ]
}
